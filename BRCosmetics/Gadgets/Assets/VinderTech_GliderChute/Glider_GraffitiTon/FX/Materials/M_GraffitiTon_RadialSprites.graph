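# Unlit translucent Niagara particle material drawing radial sprites for GraffitiTon glider effects.
# Polar-style UVs, whose arctangent takes G for both inputs, sample sprite textures and a radial mask with explicit derivatives.
# Dynamic parameters control offsets, tiling (U drives both axes) and mask shaping, without a Time node.
# The masked result indexes a curve atlas for emissive colour and sets opacity, both tinted by particle colour, with depth and camera fades on opacity.
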Material /BRCosmetics/Gadgets/Assets/VinderTech_GliderChute/Glider_GraffitiTon/FX/Materials/M_GraffitiTon_RadialSprites
graph {
 properties {
  bUsedWithNiagaraMeshParticles=true
  bUsedWithNiagaraSprites=true
  blendMode=BLEND_Translucent
  pixelDepthOffsetMode=PDOM_Legacy
  shadingModel=MSM_Unlit
  twoSided=true
  usedShadingModels=Unlit
 }
 nodes [
  n1 = ParticleColor()
  n2 = DynamicParameter() {parameterIndex=2, paramNames [Base Sprite Tile U, Base Sprite Tile V, Radial Mask Tile U, Radial Mask Tile V]}
  n3 = DynamicParameter() {
   parameterIndex=1
   defaultValue=<rgba 1 1 1 2>
   paramNames [Base Sprite Offset U, Base Sprite Offset V, Radial Mask Offset U, Radial Mask Offset V]
  }
  n4 = ComponentMask(None=@n3.Base Sprite Offset U) {r=true}
  n5 = ComponentMask(None=@n3.Base Sprite Offset V) {g=true}
  n6 = Subtract(A=TextureCoordinate()) {constB=0.5}
  n7 = SquareRoot(None=MaterialFunctionCall(Vector 2=@n6) {materialFunction=/Engine/Functions/Engine_MaterialFunctions02/Utility/VectorLength}.V2 Length)
  n8 = ScalarParameter() {defaultValue=1, parameterName=RadialMult, sliderMax=1}
  n9 = MaterialFunctionCall(Float2=@n6) {materialFunction=&BreakOutFloat2Components}
  n10 = Arctangent2Fast(Y=@n9.G, X=@n9.G)
  n11 = MaterialFunctionCall() {materialFunction=&Pi}
  n12 = Constant() {r=2}
  n13 = AppendVector(
   A=Power(Base=Multiply(A=@n7, B=@n8), Exp=ScalarParameter() {defaultValue=1, parameterName=RadialPower, sliderMax=5})
   B=Divide(
    A=Add(A=@n10, B=@n11.Result)
    B=MaterialFunctionCall(Multiplier=@n12) {materialFunction=&Pi}.Result
   )
  )
  n14 = Constant2Vector() {g=1, r=1}
  n15 = ComponentMask(None=@n14) {r=true}
  n16 = DynamicParameter() {defaultValue=<rgba 1 1 1 2>, paramNames [Sprite Extrusion Speed, Sprite Spin Speed, Radial Mask Strength, Radial Mask Power]}
  n17 = ComponentMask(None=@n14) {g=true}
  n18 = Add(
   A=AppendVector(A=Multiply(A=@n15, B=@n16.Sprite Extrusion Speed), B=Multiply(A=@n17, B=@n16.Sprite Spin Speed))
   B=Add(A=AppendVector(A=@n4, B=@n5), B=@n13)
  )
  n19 = MaterialFunctionCall(XY=@n18) {materialFunction=&Swizzle}
  n20 = ScalarParameter() {group="1) Base Sprite Rotation", parameterName=Base Sprite - Rotate UVs, sortPriority=0}
  n21 = Constant() {r=1}
  n22 = Constant() {r=-1}
  n23 = LinearInterpolate(A=@n21, B=@n22, Alpha=ScalarParameter() {group="1) Base Sprite Rotation", parameterName=Base Sprite - Flip UVs Horizontal, sortPriority=1})
  n24 = LinearInterpolate(A=@n21, B=@n22, Alpha=ScalarParameter() {group="1) Base Sprite Rotation", parameterName=Base Sprite - Flip UVs Vertically, sortPriority=2})
  n25 = Multiply(
   A=AppendVector(A=@n2.Base Sprite Tile U, B=@n2.Base Sprite Tile U)
   B=Multiply(A=LinearInterpolate(A=@n18, B=@n19.YX, Alpha=@n20), B=AppendVector(A=@n23, B=@n24))
  )
  n26 = TextureCoordinate()
  n27 = DDX(Value=@n26)
  n28 = DDY(Value=@n26)
  n29 = TextureSampleParameter2D!(UVs=@n25, "DDX(UVs)"=@n27, "DDY(UVs)"=@n28, Apply View MipBias) {
   constMipValue=0
   group=Base Sprite Texture
   mipValueMode=TMVM_Derivative
   parameterName=Base Sprite
   sortPriority=60
   texture=&T_ShieldCaustics
  }
  n30 = TextureSampleParameter2D(UVs=@n25, "DDX(UVs)"=@n27, "DDY(UVs)"=@n28) {
   constMipValue=0
   group=Base Sprite Texture
   mipValueMode=TMVM_Derivative
   parameterName=Secondary Sprite
   samplerType=SAMPLERTYPE_LinearColor
   sortPriority=60
   texture=/Game/Characters/Player/Male/Medium/Bodies/M_MED_SRDrift/Textures/T_ElectricNoise
  }
  n31 = DynamicParameter() {parameterIndex=2, paramNames [Base Sprite Tile U, Base Sprite Tile V, Radial Mask Tile U, Radial Mask Tile V]}
  n32 = Add(A=AppendVector(A=ComponentMask(None=@n3.Radial Mask Offset U) {r=true}, B=ComponentMask(None=@n3.Radial Mask Offset V) {g=true}), B=@n13)
  n33 = LinearInterpolate(
   A=@n32
   B=MaterialFunctionCall(XY=@n32) {materialFunction=&Swizzle}.YX
   Alpha=ScalarParameter() {defaultValue=1, group="1.1) Radial Mask Rotation", parameterName=Radial Mask - Rotate UVs, sortPriority=3}
  )
  n34 = Constant() {r=1}
  n35 = Constant() {r=-1}
  n36 = AppendVector(
   A=LinearInterpolate(A=@n34, B=@n35, Alpha=ScalarParameter() {group="1.1) Radial Mask Rotation", parameterName=Radial Mask - Flip UVs Horizontal, sortPriority=4})
   B=LinearInterpolate(A=@n34, B=@n35, Alpha=ScalarParameter() {group="1.1) Radial Mask Rotation", parameterName=Radial Mask - Flip UVs Vertically, sortPriority=5})
  )
  n37 = TextureSampleParameter2D(
   UVs=Multiply(A=AppendVector(A=@n31.Radial Mask Tile U, B=@n31.Radial Mask Tile U), B=Multiply(A=@n33, B=@n36))
   "DDX(UVs)"=@n27
   "DDY(UVs)"=@n28
  ) {
   constMipValue=0
   group=Radial Mask Texture
   mipValueMode=TMVM_Derivative
   parameterName=Radial Mask
   sortPriority=70
   texture=/Game/Global/Textures/Default/Masks_and_Gradients/T_Water_Top_Ramp
  }
  n38 = DynamicParameter() {defaultValue=<rgba 1 1 1 2>, paramNames [Sprite Extrusion Speed, Sprite Spin Speed, Radial Mask Strength, Radial Mask Power]}
  n39 = Multiply(
   A=LinearInterpolate(A=@n29.R, B=Add(A=@n30.R, B=@n29.R), Alpha=ScalarParameter() {parameterName=Add Secondary Sprite?})
   B=Power(Base=Multiply(A=@n37.R, B=@n38.Radial Mask Strength), Exp=@n38.Radial Mask Power)
  )
  n40 = CurveAtlasRowParameter(CurveTime=Multiply(A=SmoothStep(Value=@n39) {constMax=0.3, constMin=0.15, constValue=0.6}) {constB=0.7}) {
   atlas=/BRCosmetics/Gadgets/Assets/VinderTech_GliderChute/Glider_GraffitiTon/FX/CA_GraffitiTon
   curve=/BRCosmetics/Gadgets/Assets/VinderTech_GliderChute/Glider_GraffitiTon/FX/C_GraffitiTon
  }
  MaterialFunctionCall(Emissive=Multiply(A=@n1.RGB, B=@n40)) {materialFunction=&MF_FX_EmissiveAdjustments}
  n41 = SmoothStep(Value=@n39) {constMin=0.05, constValue=0.6}
  Multiply(
   A=DepthFade(
    Opacity=Multiply(A=@n1.A, B=@n41)
    FadeDistance=DynamicParameter() {parameterIndex=3, defaultValue=<rgba 1 1 1 2>, paramNames [Depth Fade, N/A, N/A, N/A]}.Depth Fade
   )
   B=MaterialFunctionCall(
    Fade Length=ScalarParameter() {parameterName="Camera Fade Length (S)"}
    Fade Offset=ScalarParameter() {parameterName="Camera Fade Offset (S)"}
   ) {materialFunction=&CameraDepthFade}.Result
  )
 ]
}
