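# Niagara burst effect for the GraffitiTon glider's deploy, a colorful paint-like splash.
# CenterSplash and AdditionalCenter form the expanding core, Drops and GravityChunks fling gravity-pulled sprites outward, and RadialSprite adds a three-renderer radial flash placed via skeletal mesh sampling.
# User.Variant blends between two color palettes and sets RadialSprite's Default/Neon integer flags.
# System time-of-day color tints all emitters, while front-end modifiers adjust center offsets and radial size and opacity.
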NiagaraSystem /BRCosmetics/Gadgets/Assets/VinderTech_GliderChute/Glider_GraffitiTon/FX/NS_Glider_GraffitiTon_Deploy_Burst
dependencies {
 usedDataInterfaces [
  &NiagaraDataInterfaceCurve
  &NiagaraDataInterfaceArrayColor
  &NiagaraDataInterfaceArrayFloat
  &NiagaraDataInterfaceColorCurve
 ]
 usedDynamicInputs [&SimulationPosition, &ColorFromCurve]
 usedModules [
  &SystemState
  &self:SystemSpawnScript.SystemScriptSource.SystemScriptGraph.NiagaraNodeAssignment_0.SetVariables_#1
  &self:SystemSpawnScript.SystemScriptSource.SystemScriptGraph.NiagaraNodeAssignment_1.SetVariables_#2
 ]
}
emitters {
 AdditionalCenter {
  emitterUpdateScript [
   &EmitterState()
   &SpawnBurst_Instantaneous(Spawn Count=2, Use Spawn Probability=false, UseLoopCountLimit=false)
  ]
  particleSpawnScript [
   &PartitionParticles()
   &InitializeParticle(
    Color=&LerpLinearColors(
     StartColor=&SelectLinearColorFromArray(
      Array Sampling Mode=::NewEnumerator1
      Direct Array Index=$Particles.PartitionParticles.Partition
      Color Selection Array=DataInterface() {propertyValues {ColorData [<rgba 3.4667 0.8 4 1>, <rgba 4 3.71 0.52 1>]}}
      Random Seed=?
     )
     EndColor=&SelectLinearColorFromArray(
      Array Sampling Mode=::NewEnumerator1
      Direct Array Index=$Particles.PartitionParticles.Partition
      Color Selection Array=DataInterface() {propertyValues {ColorData [<rgba 3.0208 0.25 5 1>, <rgba 0.4 2.8 4 1>]}}
      Random Seed=?
     )
     LerpFactor=$User.Variant
    )
    Lifetime Max=0.4
    Lifetime Min=0.38
    Lifetime Mode=::NewEnumerator1
    Position Offset=&FN_FrontEndVector3Modifier(Gameplay=<0 -50 15>, Front End=<0 100 0>)
    Sprite Rotation Mode=::NewEnumerator2
    Sprite Size Mode=::NewEnumerator1
    Uniform Sprite Size Max=22
    Uniform Sprite Size Min=21
    UsePositionOffset=true
    Write Material Random=true
   )
   &SubUVAnimation(End Frame=0, SubUV Animation Mode:ENiagara_SubUVLookupMode=::NewEnumerator2)
   &CameraOffset(Camera Offset Amount=15)
  ]
  particleUpdateScript [
   &ParticleState()
   &DynamicMaterialParameters(
    Index 0 Param 2=&FloatFromCurve(FloatCurve=DataInterface() {propertyValues {Curve {keys [<at 0.35 0.45 RCIM_Cubic>,<at 1 0.7 RCIM_Cubic>]}}})
    Index 0 Param 3=&FloatFromCurve(FloatCurve=DataInterface() {propertyValues {Curve {keys [<at 0.2 -0.5>,<at 0.9 -1>]}}})
    Index 0 Param 4=&FloatFromCurve(FloatCurve=DataInterface() {propertyValues {Curve {keys [{arriveTangent=2.6651, interpMode=RCIM_Cubic, leaveTangent=2.6651, tangentMode=RCTM_User, time=0, value=0.1}, <at 0.6 1.333 RCIM_Cubic>]}}})
    Module.Param0WriteEnabled=true
    Module.Param1WriteEnabled=true
    Module.Param2WriteEnabled=true
    Module.Param3WriteEnabled=true
   )
   &GravityForce(Gravity=<0 0 -300>)
   &SolveForcesAndVelocity(Clamp Velocity=false, Limit Acceleration=false)
   &ScaleColor(
    Scale Alpha=&FloatFromCurve(
     FloatCurve=DataInterface() {propertyValues {Curve {keys [{interpMode=RCIM_Cubic, tangentMode=RCTM_User, time=0.5, value=1}, {interpMode=RCIM_Cubic, tangentMode=RCTM_User, time=1, value=0}]}}}
     Scale Curve=&FNTimeOfDayScalar(Night Scale=1, Morning Scale=1, Evening Scale=1)
    )
    Scale RGB=&Multiply_VectorByFloat(
     Vector=&MakeVectorFromLinearColorRGB(Color=$System.ToDColor)
     Float=&FloatFromCurve(FloatCurve=DataInterface() {propertyValues {Curve {keys [
          {arriveTangent=-12.7849, interpMode=RCIM_Cubic, leaveTangent=-12.7849, tangentMode=RCTM_User, time=0.05, value=5}
          {interpMode=RCIM_Cubic, tangentMode=RCTM_Break, time=0.35, value=1}
         ]}}})
    )
    ScaleA=true
    ScaleRGB=true
   )
   &ScaleSpriteSize(
    Uniform Curve Scale=5
    Uniform Curve Sprite Scale=DataInterface() {propertyValues {Curve {keys [
        {interpMode=RCIM_Cubic, leaveTangent=6.6646, tangentMode=RCTM_Break, time=0, value=1}
        {arriveTangent=1.8771, interpMode=RCIM_Cubic, leaveTangent=1.8771, tangentMode=RCTM_User, time=0.4, value=2.5}
        {arriveTangent=0.9481, interpMode=RCIM_Cubic, leaveTangent=0.9481, tangentMode=RCTM_User, time=0.9983, value=3.3515}
       ]}}}
   )
  ]
  rendererClasses [&NiagaraSpriteRendererProperties]
  renderers [{rendererClass=&NiagaraSpriteRendererProperties}]
 }
 CenterSplash {
  emitterUpdateScript [
   &EmitterState()
   &SpawnBurst_Instantaneous(Use Spawn Probability=false, UseLoopCountLimit=false)
   /SetVariables(Emitter.EmitterRandom:NiagaraFloat=&RandomRangeFloat(Minimum=0, Maximum=1)) {isSetParameters=true}
  ]
  particleSpawnScript [
   &InitializeParticle(
    Color=&LerpLinearColors(StartColor=<rgba 0.33 1.1 0.9075 1>, EndColor=<rgba 0.2025 0.2025 0.9 1>, LerpFactor=$User.Variant)
    Lifetime Max=0.28
    Lifetime Min=0.27
    Lifetime Mode=::NewEnumerator1
    Position Offset=&FN_FrontEndVector3Modifier(Gameplay=<0 -50 15>, Front End=<0 100 0>)
    Sprite Rotation Mode=::NewEnumerator2
    Sprite Size Mode=::NewEnumerator1
    Uniform Sprite Size Max=20
    Uniform Sprite Size Min=19
    UsePositionOffset=true
    Write Material Random=true
   )
   &CameraOffset(Camera Offset Amount=20)
  ]
  particleUpdateScript [
   &ParticleState()
   &DynamicMaterialParameters(
    Index 0 Param 1=$Emitter.EmitterRandom
    Index 0 Param 3=&FloatFromCurve(FloatCurve=DataInterface() {propertyValues {Curve {keys [
         {interpMode=RCIM_Cubic, tangentMode=RCTM_Break, time=0.3, value=-0.6}
         {arriveTangent=-0.0609, interpMode=RCIM_Cubic, tangentMode=RCTM_Break, time=0.95, value=-1.3}
        ]}}})
    Index 0 Param 4=&FloatFromCurve(FloatCurve=DataInterface() {propertyValues {Curve {keys [{arriveTangent=2.6651, interpMode=RCIM_Cubic, leaveTangent=2.6651, tangentMode=RCTM_User, time=0, value=0.1}, <at 0.6 1.333 RCIM_Cubic>]}}})
    Module.Param0WriteEnabled=true
    Module.Param1WriteEnabled=true
    Module.Param2WriteEnabled=true
    Module.Param3WriteEnabled=true
   ) {module=DynamicMaterialParameters001}
   &GravityForce(Gravity=<0 0 -300>)
   &SolveForcesAndVelocity(Clamp Velocity=false, Limit Acceleration=false)
   &ScaleColor(
    Scale Alpha=&FloatFromCurve(FloatCurve=DataInterface() {propertyValues {Curve {keys [{interpMode=RCIM_Cubic, tangentMode=RCTM_User, time=0.8, value=1}, {interpMode=RCIM_Cubic, tangentMode=RCTM_User, time=1, value=0}]}}})
    Scale RGB=&Multiply_VectorByFloat(
     Vector=&MakeVectorFromLinearColorRGB(Color=$System.ToDColor)
     Float=&MakeFloatFromVector(
      Vector=&VectorFromFloat(Value=&FloatFromCurve(FloatCurve=DataInterface() {propertyValues {Curve {keys [
            {arriveTangent=-12.7849, interpMode=RCIM_Cubic, leaveTangent=-12.7849, tangentMode=RCTM_User, time=0.05, value=3}
            {interpMode=RCIM_Cubic, tangentMode=RCTM_Break, time=0.35, value=1}
           ]}}}))
      Channel=::NewEnumerator0
     )
    )
    ScaleA=true
    ScaleRGB=true
   )
   &ScaleSpriteSize(
    Uniform Curve Scale=4.125
    Uniform Curve Sprite Scale=DataInterface() {propertyValues {Curve {keys [
        {interpMode=RCIM_Cubic, leaveTangent=7.3645, tangentMode=RCTM_Break, time=0, value=1}
        {arriveTangent=2.1362, interpMode=RCIM_Cubic, leaveTangent=2.1362, tangentMode=RCTM_User, time=0.4, value=2.875}
        {arriveTangent=0.9481, interpMode=RCIM_Cubic, leaveTangent=0.9481, tangentMode=RCTM_User, time=0.9983, value=3.3515}
       ]}}}
   )
  ]
  rendererClasses [&NiagaraSpriteRendererProperties]
  renderers [{rendererClass=&NiagaraSpriteRendererProperties}]
 }
 Drops {
  emitterUpdateScript [
   &EmitterState()
   &SpawnBurst_Instantaneous(Spawn Count=&RandomRangeInt(Minimum=8, Maximum=16), Use Spawn Probability=false)
   &SpawnBurst_Instantaneous(Spawn Count=&RandomRangeInt(Minimum=14, Maximum=20), Spawn Group=1, Use Spawn Probability=false) {module=SpawnBurst_Instantaneous001}
  ]
  particleSpawnScript [
   &InitializeParticle(
    Color=&LerpLinearColors(
     StartColor=&SelectLinearColorFromArray(
      Array Sampling Mode=::NewEnumerator0
      Direct Array Index=?
      Color Selection Array=DataInterface() {propertyValues {ColorData [<rgba 3 0.6 3 1>, <rgba 4 4 0.52 1>, <rgba 0.8 2 1.4 1>]}}
      Random Seed=0
     )
     EndColor=&SelectLinearColorFromArray(
      Array Sampling Mode=::NewEnumerator0
      Direct Array Index=?
      Color Selection Array=DataInterface() {propertyValues {ColorData [<rgba 2.625 0.25 5 1>, <rgba 0.4 2.2 4 1>, <rgba 0.27 0.27 1.2 1>]}}
      Random Seed=0
     )
     LerpFactor=$User.Variant
    )
    Lifetime=&UniformRangedFloat(Minimum=0.4, Maximum=1)
    Mass Mode=::NewEnumerator1
    Sprite Size=&Multiply_Vector2D_ByFloat(
     Float=&UniformAOrBFloat(
      A=&RandomRangeFloat(Minimum=2.5, Maximum=6)
      B=7
      Distribution Weight=0.1
      Evaluation Type=::NewEnumerator0
      Recalculate Random Each Loop=true
      Override Seed=false
     )
     Vector2D=&UniformRangedVector2D(Minimum=<1 3>, Maximum=<1 4>)
    )
    Sprite Size Mode=::NewEnumerator4
    UsePositionOffset=false
    Write Material Random=true
   ) {module=InitializeParticle001}
   /Niagara/Modules/Spawn/Location/SphereLocation(MaskSpawnGroup=false, Module.Override Randomness=false, Sphere Radius=5, Surface Only=false)
   &AddVelocityFromPoint(
    Module.Use Falloff Distance=false
    Velocity Strength=&Multiply_Float(A=&UniformRangedFloat(Minimum=250, Maximum=500), B=2)
   )
   /Niagara/Modules/Update/SubUV/SubUVAnimation(SubUV Animation Mode=::NewEnumerator1) {module=SubUVAnimation001}
  ]
  particleUpdateScript [
   &ParticleState()
   &Drag(Use Linear Drag=true, Use Rotational Drag=true)
   &GravityForce(Gravity=&MakeVector(X=0, Y=0, Z=&RandomRangeFloat(Minimum=-700, Maximum=-980))) {module=GravityForce001}
   &SolveForcesAndVelocity(Clamp Velocity=false, Limit Acceleration=false)
   &ScaleSpriteSize(Scale Factor=&Multiply_Vector2D_ByFloat(
     Float=&FloatFromCurve(
      FloatCurve=DataInterface() {propertyValues {Curve {keys [<at 0 0.5>,<at 1 1>]}}}
      Scale Curve=&SelectFloatFromArray(Direct Array Index=&SpawnGroup(), Float Selection Array=DataInterface() {propertyValues {FloatData [2.5,1]}})
     )
     Vector2D=<1 1>
    ))
   &ScaleColor(
    Scale Alpha=&FloatFromCurve(FloatCurve=DataInterface() {propertyValues {Curve {keys [<at 0 1>,<at 0.75 1>,<at 1 0>]}}})
    Scale RGB=&Multiply_VectorByFloat(
     Vector=&MakeVectorFromLinearColorRGB(Color=$System.ToDColor)
     Float=&FloatFromCurve(FloatCurve=DataInterface() {propertyValues {Curve {keys [
          {arriveTangent=-12.7849, interpMode=RCIM_Cubic, leaveTangent=-12.7849, tangentMode=RCTM_User, time=0.05, value=3}
          {interpMode=RCIM_Cubic, tangentMode=RCTM_Break, time=0.35, value=1}
         ]}}})
    )
    ScaleA=true
    ScaleRGB=true
   )
   &DynamicMaterialParameters(
    Index 0 Param 1=&RandomRangeFloat(Minimum=0, Maximum=1)
    Index 0 Param 2=&FloatFromCurve(FloatCurve=DataInterface() {propertyValues {Curve {keys [<at 0.35 0.45>,<at 1 0.9>]}}})
    Index 0 Param 3=&FloatFromCurve(FloatCurve=DataInterface() {propertyValues {Curve {keys [{arriveTangent=2.6651, interpMode=RCIM_Cubic, leaveTangent=2.6651, tangentMode=RCTM_User, time=0, value=0}, <at 1 1 RCIM_Cubic>]}}})
    Index 0 Param 4=&FloatFromCurve(FloatCurve=DataInterface() {propertyValues {Curve {keys [<at 0 0 RCIM_Cubic>, {arriveTangent=2.332, interpMode=RCIM_Cubic, leaveTangent=2.332, tangentMode=RCTM_User, time=1, value=1}]}}})
    Module.Param0WriteEnabled=true
    Module.Param1WriteEnabled=true
    Module.Param2WriteEnabled=false
    Module.Param3WriteEnabled=true
   ) {module=DynamicMaterialParameters001}
  ]
  rendererClasses [&NiagaraSpriteRendererProperties]
  renderers [{rendererClass=&NiagaraSpriteRendererProperties}]
 }
 GravityChunks {
  emitterUpdateScript [
   &EmitterState()
   &SpawnBurst_Instantaneous(
    Spawn Count=&RandomRangeInt(Minimum=9, Maximum=16)
    Use Spawn Probability=false
    UseLoopCountLimit=false
   )
  ]
  particleSpawnScript [
   &InitializeParticle(
    Color=&LerpLinearColors(
     StartColor=&SelectLinearColorFromArray(
      Array Sampling Mode=::NewEnumerator0
      Direct Array Index=?
      Color Selection Array=DataInterface() {propertyValues {ColorData [<rgba 3 0.6 3 1>, <rgba 4 4 0.52 1>, <rgba 0.8 2 1.4 1>]}}
      Random Seed=0
     )
     EndColor=&SelectLinearColorFromArray(
      Array Sampling Mode=::NewEnumerator0
      Direct Array Index=?
      Color Selection Array=DataInterface() {propertyValues {ColorData [<rgba 2.625 0.25 5 1>, <rgba 0.4 2.2 4 1>, <rgba 0.27 0.27 1.2 1>]}}
      Random Seed=0
     )
     LerpFactor=$User.Variant
    )
    Lifetime Max=0.8
    Lifetime Min=0.6
    Lifetime Mode=::NewEnumerator1
    Sprite Rotation Mode=::NewEnumerator2
    Sprite Size Mode=::NewEnumerator1
    Sprite UV Mode=::NewEnumerator3
    Uniform Sprite Size Max=45
    Uniform Sprite Size Min=40
    UsePositionOffset=false
    Write Material Random=true
   )
   /Niagara/Modules/Update/SubUV/V2/SubUVAnimation(End Frame=3, SubUV Animation Mode:ENiagara_SubUVLookupMode=::NewEnumerator2)
   &ShapeLocation(
    MaskSpawnGroup=false
    Offset Mode=::NewEnumerator0
    Override Seed=false
    Sphere Radius=&RandomRangeFloat(Minimum=15, Maximum=35)
    UseOwnerScale=true
   )
   &CameraOffset(Camera Offset Amount=15)
   &AddVelocity(
    Module.Use Falloff Distance=false
    Origin Offset=<0 0 -20>
    Override Seed=false
    Velocity Mode=::NewEnumerator1
    Velocity Speed=&RandomRangeFloat(Minimum=200, Maximum=500)
   )
   /SetVariables(Particles.Random=&UniformAOrBFloat(A=-1)) {isSetParameters=true}
  ]
  particleUpdateScript [
   &ParticleState()
   &SpriteRotationRate(Rotation Rate=&Multiply_Float(A=&RandomRangeFloat(Minimum=-1, Maximum=1), B=60))
   &DynamicMaterialParameters(
    Index 0 Param 2=&FloatFromCurve(FloatCurve=DataInterface() {propertyValues {Curve {keys [{interpMode=RCIM_Cubic, tangentMode=RCTM_Break, time=0.05, value=0.3}, {interpMode=RCIM_Cubic, tangentMode=RCTM_Break, time=0.95, value=1}]}}})
    Index 0 Param 3=&FloatFromCurve(FloatCurve=DataInterface() {propertyValues {Curve {keys [
         {interpMode=RCIM_Cubic, tangentMode=RCTM_Break, time=0.05, value=-0.5}
         {interpMode=RCIM_Cubic, tangentMode=RCTM_Break, time=0.95, value=-1.43}
        ]}}})
    Index 0 Param 4=&FloatFromCurve(FloatCurve=DataInterface() {propertyValues {Curve {keys [{arriveTangent=2.6651, interpMode=RCIM_Cubic, leaveTangent=2.6651, tangentMode=RCTM_User, time=0, value=0.1}, <at 0.6 1.333 RCIM_Cubic>]}}})
    Module.Param0WriteEnabled=true
    Module.Param1WriteEnabled=true
    Module.Param2WriteEnabled=true
    Module.Param3WriteEnabled=true
   )
   &GravityForce(Gravity=&MakeVector(X=0, Y=0, Z=&RandomRangeFloat(Minimum=-800, Maximum=-1000)))
   &SolveForcesAndVelocity(Clamp Velocity=false, Limit Acceleration=false)
   &ScaleColor(
    Scale Alpha=&FloatFromCurve(FloatCurve=DataInterface() {propertyValues {Curve {keys [
         {interpMode=RCIM_Cubic, tangentMode=RCTM_Break, time=0, value=0}
         {interpMode=RCIM_Cubic, tangentMode=RCTM_Break, time=0.15, value=1}
         {interpMode=RCIM_Cubic, tangentMode=RCTM_User, time=0.5, value=1}
         {interpMode=RCIM_Cubic, tangentMode=RCTM_User, time=1, value=0}
        ]}}})
    Scale RGB=&Multiply_VectorByFloat(
     Vector=&MakeVectorFromLinearColorRGB(Color=$System.ToDColor)
     Float=&FloatFromCurve(FloatCurve=DataInterface() {propertyValues {Curve {keys [<at 0.05 3>,<at 0.25 1>]}}})
    )
    ScaleA=false
    ScaleRGB=true
   )
   &ScaleSpriteSize(
    Uniform Curve Scale=2
    Uniform Curve Sprite Scale=DataInterface() {propertyValues {Curve {keys [<at 0 0.5>, {arriveTangent=0.9481, leaveTangent=0.9481, time=0.9983, value=1}]}}}
   )
   &SpriteRotationRate(Rotation Rate=&Multiply_Float(
     A=&Multiply_Float(A=&RandomRangeFloat(Minimum=1, Maximum=0.5), B=$Particles.Random)
     B=&FloatFromCurve(
      FloatCurve=DataInterface() {propertyValues {Curve {keys [{arriveTangent=-2, interpMode=RCIM_Cubic, time=0.15, value=1}, {interpMode=RCIM_Cubic, leaveTangent=-8.0089, time=0.8, value=0}]}}}
      Scale Curve=300
     )
    )) {module=SpriteRotationRate001}
  ]
  rendererClasses [&NiagaraSpriteRendererProperties]
  renderers [{rendererClass=&NiagaraSpriteRendererProperties}]
 }
 RadialSprite {
  emitterUpdateScript [
   &EmitterState()
   &SpawnBurst_Instantaneous(Spawn Count=1, Use Spawn Probability=false)
  ]
  particleSpawnScript [&InitializeParticle(
    Lifetime Max=0.35
    Lifetime Min=0.3
    Lifetime Mode=::NewEnumerator1
    Sprite Size Mode=::NewEnumerator1
    Uniform Sprite Size Max=60
    Uniform Sprite Size Min=60
    UsePositionOffset=false
    Write Material Random=true
   )]
  particleUpdateScript [
   &ParticleState()
   &SkeletalMeshLocation(
    Inherit Velocity=false
    MaskSpawnGroup=false
    Mesh Sampling Type=::NewEnumerator5
    Sampled Orientation Coordinate Space=::Local
    Sampled Position Offset Coordinate Space:ENiagaraMeshTransforms=::NewEnumerator3
    Sampled Velocity Coordinate Space:ENiagaraCoordinateSpace=::Local
    Skeletal Mesh=$System.Skeletal Mesh
    Socket Sampling Mode=::NewEnumerator5
    Use Sampled Mesh Attribute as Ribbon ID=false
   )
   &FN_Radial_Material(
    Base Sprite Offset X:NiagaraFloat=&RandomRangeFloat(Minimum=0, Maximum=1)
    Base Sprite Offset Y:NiagaraFloat=&RandomRangeFloat(Minimum=0, Maximum=1)
    Base Sprite Tile U=2
    Depth Fade=5
    Radial Mask Strength=0.8
    Sprite Extrusion Speed:NiagaraFloat=&RandomRangeFloat(Minimum=-0.2, Maximum=-0.4)
    Sprite Spin Speed:NiagaraFloat=&RandomRangeFloat(Minimum=0, Maximum=0)
   )
   &ScaleSpriteSize(Scale Factor=&Vector2DFromFloat(Value=&FloatFromCurve(
      FloatCurve=DataInterface() {propertyValues {Curve {keys [
          {arriveTangent=1.6226, interpMode=RCIM_Cubic, leaveTangent=1.6226, tangentMode=RCTM_User, time=0, value=0.5}
          {interpMode=RCIM_Cubic, tangentMode=RCTM_User, time=1, value=1}
         ]}}}
      Scale Curve=5.5
     )))
   &ScaleColor(
    Scale Alpha=&FloatFromCurve(
     FloatCurve=DataInterface() {propertyValues {Curve {keys [
         {interpMode=RCIM_Cubic, tangentMode=RCTM_User, time=0, value=0}
         {arriveTangent=-0.2108, interpMode=RCIM_Cubic, leaveTangent=-0.2108, tangentMode=RCTM_User, time=0.2, value=1.1}
         {arriveTangent=-0.0679, interpMode=RCIM_Cubic, leaveTangent=-0.0679, tangentMode=RCTM_User, time=1, value=0}
        ]}}}
     Scale Curve=&Lerp_Float(A=&FN_FrontEndFloatModifier(Gameplay=6, Front End=3), B=10, Alpha=$User.Variant)
    )
    Scale RGB=&MakeVectorFromLinearColorRGB(Color=$System.ToDColor)
    ScaleA=true
    ScaleRGB=true
   )
   /SetVariables(Particles.Neon:NiagaraInt32=&MakeCustomIntFromBool(Bool=&SetBoolByFloatComparison(A=$User.Variant, B=0), True INT=0, False INT=1)) {isSetParameters=true}
   /SetVariables(Particles.Default:NiagaraInt32=&MakeCustomIntFromBool(Bool=&SetBoolByFloatComparison(A=$User.Variant, B=0), True INT=1, False INT=0)) {isSetParameters=true}
   /SetVariables(
    Particles.SpriteColor_02:LinearColor=&MultiplyLinearColor(
     Linear Color=&LerpLinearColors(StartColor=<rgba 0.0395 0.2016 0.1746 1>, EndColor=<rgba 0.0202 0.0202 0.2016 1>, LerpFactor=$User.Variant)
     Scale Factor=&MakeFloat4FromVectorAndFloat(
      "Vector (RGB)"=<1 1 1>
      "Float (Alpha)"=&FloatFromCurve(FloatCurve=DataInterface() {propertyValues {Curve {keys [
           {arriveTangent=1.6226, interpMode=RCIM_Cubic, leaveTangent=1.6226, tangentMode=RCTM_User, time=0, value=0.5}
           {interpMode=RCIM_Cubic, tangentMode=RCTM_User, time=1, value=1}
          ]}}})
     )
    )
    Particles.SpriteSize_02:Vector2f=&Multiply_Vector2(
     A=<30 30>
     B=&Vector2DFromFloat(Value=&FloatFromCurve(
       FloatCurve=DataInterface() {propertyValues {Curve {keys [
           <at 0 0 RCIM_Cubic>
           {arriveTangent=2, interpMode=RCIM_Cubic, leaveTangent=0.0802, tangentMode=RCTM_Break, time=0.1, value=1}
           {arriveTangent=-0.1222, interpMode=RCIM_Cubic, leaveTangent=-1.25, tangentMode=RCTM_Break, time=0.5, value=1}
           <at 1 0 RCIM_Cubic>
          ]}}}
       Scale Curve=&FN_FrontEndFloatModifier(Gameplay=7.5, Front End=3)
      ))
    )
   ) {isSetParameters=true}
  ]
  rendererClasses [&NiagaraSpriteRendererProperties]
  renderers [
   {rendererClass=&NiagaraSpriteRendererProperties}
   {rendererClass=&NiagaraSpriteRendererProperties, rendererIndex=1}
   {rendererClass=&NiagaraSpriteRendererProperties, rendererIndex=2}
  ]
 }
}
summary {
 systemName=NS_Glider_GraffitiTon_Deploy_Burst
 emitters [
  {emitterName=CenterSplash, rendererClasses [&NiagaraSpriteRendererProperties]}
  {emitterName=GravityChunks, rendererClasses [&NiagaraSpriteRendererProperties]}
  {emitterName=Drops, rendererClasses [&NiagaraSpriteRendererProperties]}
  {emitterName=AdditionalCenter, rendererClasses [&NiagaraSpriteRendererProperties]}
  {emitterName=RadialSprite, rendererClasses [&NiagaraSpriteRendererProperties]}
 ]
 userVariables [
  User.bIsFrontEnd {type=NiagaraBool}
  User.bIsFrontEndPreview {type=NiagaraBool}
  User.SkeletalMesh {
   type=NiagaraDataInterfaceSkeletalMesh
   defaultValue {
    struct=NiagaraExt_VariableValue_DataInterface
    value {dataInterface=&self:NiagaraDataInterfaceSkeletalMesh_1}
   }
  }
  User.Variant {type=NiagaraFloat, defaultValue=0, description=""}
 ]
}
