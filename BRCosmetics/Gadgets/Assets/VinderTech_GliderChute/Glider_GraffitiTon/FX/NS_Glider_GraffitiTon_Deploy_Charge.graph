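# Niagara charge-up effect for the GraffitiTon glider deploy, converging paint-like particles on a sampled skeletal mesh point.
# BigSplash and Blobs sprites spawn in spheres around that point and are attracted inward, with delay and duration adjusted for front end.
# MeshCards and a three-renderer RadialSprite follow the point, spinning and shrinking through animated material parameters.
# User.Variant blends palettes and sets Default/Neon flags, while System.ToDColor tints every emitter.
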NiagaraSystem /BRCosmetics/Gadgets/Assets/VinderTech_GliderChute/Glider_GraffitiTon/FX/NS_Glider_GraffitiTon_Deploy_Charge
dependencies {
 usedDataInterfaces [
  &NiagaraDataInterfaceArrayColor
  &NiagaraDataInterfaceCurve
  &NiagaraDataInterfaceColorCurve
  &NiagaraDataInterfaceSkeletalMesh
 ]
 usedDynamicInputs [&ColorFromCurve]
 usedModules [
  &self:SystemSpawnScript.SystemScriptSource.SystemScriptGraph.NiagaraNodeAssignment_0.SetVariables_#1
  &SystemState
  &self:SystemSpawnScript.SystemScriptSource.SystemScriptGraph.NiagaraNodeAssignment_1.SetVariables_#2
  &self:SystemSpawnScript.SystemScriptSource.SystemScriptGraph.NiagaraNodeAssignment_2.SetVariables_#3
 ]
}
emitters {
 BigSplash {
  emitterUpdateScript [
   &EmitterState(
    Life Cycle Mode=::NewEnumerator1
    Loop Behavior=::NewEnumerator1
    Loop Delay=&FN_FrontEndFloatModifier(Gameplay=0.42, Front End=0)
    Loop Duration=&FN_FrontEndFloatModifier(Gameplay=0.5, Front End=0.4)
    UseLoopDelay=true
   )
   &SpawnRate(SpawnRate=55, Use Spawn Probability=false)
  ]
  particleSpawnScript [
   &InitializeParticle(
    Color=&LerpLinearColors(
     StartColor=&SelectLinearColorFromArray(
      Array Sampling Mode=::NewEnumerator0
      Direct Array Index=?
      Color Selection Array=DataInterface() {propertyValues {ColorData [<rgba 3 0.6 3 1>, <rgba 4 4 0.52 1>, <rgba 0.8 2 1.4 1>]}}
      Random Seed=0
     )
     EndColor=&SelectLinearColorFromArray(
      Array Sampling Mode=::NewEnumerator0
      Direct Array Index=?
      Color Selection Array=DataInterface() {propertyValues {ColorData [<rgba 2.625 0.25 5 1>, <rgba 0.4 2.2 4 1>, <rgba 0.27 0.27 1.2 1>]}}
      Random Seed=0
     )
     LerpFactor=$User.Variant
    )
    Lifetime=&RandomRangeFloat(Minimum=0.4, Maximum=0.5)
    Sprite Rotation Mode=::NewEnumerator2
    Sprite Size Mode=::NewEnumerator3
    Uniform Sprite Size=&RandomRangeFloat(Minimum=35, Maximum=50)
    UsePositionOffset=false
    Write Material Random=true
   ) {module=InitializeParticle001}
   &SkeletalMeshLocation(
    Inherit Velocity=false
    MaskSpawnGroup=false
    Mesh Sampling Type=::NewEnumerator5
    Skeletal Mesh=$System.Skeletal Mesh
    Socket Sampling Mode=::NewEnumerator5
   )
   &SphereLocation(
    MaskSpawnGroup=false
    Override Seed=false
    Sphere Origin:Vector3f=$Output.SkeletalMeshLocation.SampledPosition
    Sphere Radius=200
    Surface Only=false
    Surface Only Band Thickness=0.1
    UseNonUniformScale=false
    UseOffset=false
    UseSphereOrientation=false
   )
  ]
  particleUpdateScript [
   &ParticleState()
   &SkeletalMeshLocation(
    Apply To Particle Position=false
    Inherit Velocity=false
    MaskSpawnGroup=false
    Mesh Sampling Type=::NewEnumerator5
    Skeletal Mesh=$System.Skeletal Mesh
    Socket Sampling Mode=::NewEnumerator5
   ) {module=SkeletalMeshLocation001}
   &PointAttractionForce(
    Attraction Radius=1500
    Attraction Strength=5
    Attractor Position:Vector3f=$Output.SkeletalMeshLocation001.SampledPosition
    Falloff Exponent=1.2
    Kill Within Radius=false
    Module.Use Falloff=true
    UseAttractorOffset=false
   )
   &ScaleSpriteSize(Scale Factor=&Vector2DFromFloat(Value=&FloatFromCurve(
      FloatCurve=DataInterface() {propertyValues {Curve {keys [{interpMode=RCIM_Cubic, tangentMode=RCTM_User, time=0, value=1}, {interpMode=RCIM_Cubic, tangentMode=RCTM_User, time=1, value=0.7}]}}}
      Scale Curve=2
     )))
   &SolveForcesAndVelocity(Clamp Velocity=false, Limit Acceleration=false)
   &ScaleColor(
    Scale Alpha=&FloatFromCurve(FloatCurve=DataInterface() {propertyValues {Curve {keys [
         {arriveTangent=-0.4479, interpMode=RCIM_Cubic, leaveTangent=-0.4479, tangentMode=RCTM_User, time=0, value=0}
         {
          arriveTangent=4.0161
          arriveTangentWeight=0.0997
          interpMode=RCIM_Cubic
          leaveTangent=4.0161
          leaveTangentWeight=0.3488
          tangentMode=RCTM_User
          time=0.3
          value=1.225
         }
         {
          arriveTangent=-2.5409
          arriveTangentWeight=0.3735
          interpMode=RCIM_Cubic
          leaveTangent=-2.5409
          leaveTangentWeight=0.0534
          tangentMode=RCTM_User
          time=0.7
          value=1.925
         }
         <at 1 0 RCIM_Cubic>
        ]}}})
    Scale RGB=&MakeVectorFromLinearColorRGB(Color=$System.ToDColor)
    ScaleA=true
    ScaleRGB=true
   )
   &SpriteRotationRate(Rotation Rate=&RandomRangeFloat(Minimum=-250, Maximum=-150))
   &DynamicMaterialParameters(
    Index 0 Param 2=&FloatFromCurve(FloatCurve=DataInterface() {propertyValues {Curve {keys [
         {arriveTangentWeight=0.3, interpMode=RCIM_Cubic, leaveTangentWeight=0.3, tangentMode=RCTM_Break, time=0.05, value=1}
         {arriveTangentWeight=0.3, interpMode=RCIM_Cubic, leaveTangentWeight=0.3, tangentMode=RCTM_Break, time=0.95, value=0.3}
        ]}}})
    Index 0 Param 3=&FloatFromCurve(FloatCurve=DataInterface() {propertyValues {Curve {keys [
         {interpMode=RCIM_Cubic, tangentMode=RCTM_Break, time=0.05, value=-0.5}
         {interpMode=RCIM_Cubic, tangentMode=RCTM_Break, time=0.95, value=-1.43}
        ]}}})
    Index 0 Param 4=&FloatFromCurve(FloatCurve=DataInterface() {propertyValues {Curve {keys [{arriveTangent=2.6651, interpMode=RCIM_Cubic, leaveTangent=2.6651, tangentMode=RCTM_User, time=0, value=0.1}, <at 0.6 1.333 RCIM_Cubic>]}}})
    Module.Param0WriteEnabled=true
    Module.Param1WriteEnabled=true
    Module.Param2WriteEnabled=true
    Module.Param3WriteEnabled=true
   )
  ]
  rendererClasses [&NiagaraSpriteRendererProperties]
  renderers [{rendererClass=&NiagaraSpriteRendererProperties}]
 }
 Blobs {
  emitterUpdateScript [
   &EmitterState(
    Life Cycle Mode=::NewEnumerator1
    Loop Behavior=::NewEnumerator1
    Loop Delay=&FN_FrontEndFloatModifier(Gameplay=0.42, Front End=0)
    Loop Duration=&FN_FrontEndFloatModifier(Gameplay=0.8, Front End=0.7)
    UseLoopDelay=true
   )
   &SpawnRate(SpawnRate=60, Use Spawn Probability=false)
  ]
  particleSpawnScript [
   &InitializeParticle(
    Color=&LerpLinearColors(
     StartColor=&SelectLinearColorFromArray(
      Array Sampling Mode=::NewEnumerator0
      Direct Array Index=?
      Color Selection Array=DataInterface() {propertyValues {ColorData [<rgba 3 0.6 3 1>, <rgba 4 4 0.52 1>, <rgba 0.8 2 1.4 1>]}}
      Random Seed=0
     )
     EndColor=&SelectLinearColorFromArray(
      Array Sampling Mode=::NewEnumerator0
      Direct Array Index=?
      Color Selection Array=DataInterface() {propertyValues {ColorData [<rgba 2.625 0.25 5 1>, <rgba 0.4 2.2 4 1>, <rgba 0.27 0.27 1.2 1>]}}
      Random Seed=0
     )
     LerpFactor=$User.Variant
    )
    Lifetime Max=0.4
    Lifetime Min=0.3
    Lifetime Mode=::NewEnumerator1
    Sprite Size=&Multiply_Vector2(A=&RandomRangeVector2D(Minimum=<4 8>, Maximum=<6 12>), B=<1.5 1.75>)
    Sprite Size Mode=::NewEnumerator4
    UsePositionOffset=false
    Write Material Random=true
   ) {module=InitializeParticle001}
   &SkeletalMeshLocation(
    Apply To Particle Position=false
    Inherit Velocity=false
    MaskSpawnGroup=false
    Mesh Sampling Type=::NewEnumerator5
    Skeletal Mesh=$System.Skeletal Mesh
    Socket Sampling Mode=::NewEnumerator5
   ) {module=SkeletalMeshLocation001}
   /Niagara/Modules/Spawn/Location/SphereLocation(
    MaskSpawnGroup=false
    Module.Override Randomness=false
    Sphere Origin:Vector3f=$Output.SkeletalMeshLocation001.SampledPosition
    Sphere Radius=225
    Surface Only=false
   )
  ]
  particleUpdateScript [
   &ParticleState()
   &SkeletalMeshLocation(
    Apply To Particle Position=false
    Inherit Velocity=false
    MaskSpawnGroup=false
    Mesh Sampling Type=::NewEnumerator5
    Skeletal Mesh=$System.Skeletal Mesh
    Socket Sampling Mode=::NewEnumerator5
   )
   &PointAttractionForce(
    Attraction Radius=1500
    Attraction Strength=8.5
    Attractor Position:Vector3f=$Output.SkeletalMeshLocation.SampledPosition
    Falloff Exponent=1.2
    Kill Within Radius=false
    Module.Use Falloff=true
    UseAttractorOffset=false
   )
   &SolveForcesAndVelocity(Clamp Velocity=false, Limit Acceleration=false)
   &ScaleSpriteSize(Scale Factor=&Vector2DFromFloat(Value=&FloatFromCurve(FloatCurve=DataInterface() {propertyValues {Curve {keys [
          {interpMode=RCIM_Cubic, tangentMode=RCTM_User, time=0, value=0}
          {interpMode=RCIM_Cubic, tangentMode=RCTM_User, time=0.2, value=1}
          {tangentMode=RCTM_User, time=1, value=0}
         ]}}})))
   &ScaleSpriteSizeBySpeed(Max Scale Factor=<1.5 5>)
   &ScaleColor(
    Scale Alpha=&FloatFromCurve(FloatCurve=DataInterface() {propertyValues {Curve {keys [
         {arriveTangent=9.0725, interpMode=RCIM_Cubic, leaveTangent=9.0725, tangentMode=RCTM_User, time=0, value=1}
         {arriveTangent=-2.5249, interpMode=RCIM_Cubic, leaveTangent=-2.5249, tangentMode=RCTM_User, time=0.1408, value=2}
         {arriveTangent=-0.0405, interpMode=RCIM_Cubic, leaveTangent=-0.0405, tangentMode=RCTM_User, time=1, value=1}
        ]}}})
    Scale RGB=&MakeVectorFromLinearColorRGB(Color=$System.ToDColor)
    ScaleA=true
    ScaleRGB=true
   )
   &DynamicMaterialParameters(
    Index 0 Param 1=&RandomRangeFloat(Minimum=0, Maximum=1)
    Index 0 Param 2=&FloatFromCurve(FloatCurve=DataInterface() {propertyValues {Curve {keys [<at 0.35 0.45>,<at 1 0.9>]}}})
    Index 0 Param 3=&FloatFromCurve(FloatCurve=DataInterface() {propertyValues {Curve {keys [{arriveTangent=2.6651, interpMode=RCIM_Cubic, leaveTangent=2.6651, tangentMode=RCTM_User, time=0, value=0}, <at 1 1 RCIM_Cubic>]}}})
    Index 0 Param 4=&FloatFromCurve(FloatCurve=DataInterface() {propertyValues {Curve {keys [<at 0 0 RCIM_Cubic>, {arriveTangent=2.332, interpMode=RCIM_Cubic, leaveTangent=2.332, tangentMode=RCTM_User, time=1, value=1}]}}})
    Module.Param0WriteEnabled=true
    Module.Param1WriteEnabled=true
    Module.Param2WriteEnabled=false
    Module.Param3WriteEnabled=true
   ) {module=DynamicMaterialParameters001}
  ]
  rendererClasses [&NiagaraSpriteRendererProperties]
  renderers [{rendererClass=&NiagaraSpriteRendererProperties}]
 }
 MeshCards {
  emitterSpawnScript [/SetVariables(Emitter.EmitterRand:NiagaraFloat=&RandomRangeFloat(Minimum=0, Maximum=360)) {isSetParameters=true}]
  emitterUpdateScript [
   &EmitterState()
   &SpawnBurst_Instantaneous(Spawn Count=3, Spawn Time=0.025, Use Spawn Probability=false)
  ]
  particleSpawnScript [
   &FN_SetSpawnGroup(Number of Spawn Groups=3)
   &InitializeParticle(
    Color=&LerpLinearColors(StartColor=<rgba 0.27 1.2 0.9985 1>, EndColor=<rgba 0.24 0.24 1.2 1>, LerpFactor=$User.Variant)
    Lifetime=0.9
    Mesh Scale Mode=::NewEnumerator3
    Mesh Uniform Scale=8
    Sprite Size Mode=::NewEnumerator1
    Uniform Sprite Size Max=60
    Uniform Sprite Size Min=60
    UsePositionOffset=false
    Write Material Random=true
   )
   &SkeletalMeshLocation(
    Apply To Particle Position=false
    Inherit Velocity=false
    MaskSpawnGroup=false
    Mesh Sampling Type=::NewEnumerator5
    Sampled Orientation Coordinate Space=::Local
    Sampled Position Offset Coordinate Space:ENiagaraMeshTransforms=::NewEnumerator3
    Sampled Velocity Coordinate Space:ENiagaraCoordinateSpace=::Local
    Skeletal Mesh=$System.Skeletal Mesh
    Socket Sampling Mode=::NewEnumerator5
    Use Sampled Mesh Attribute as Ribbon ID=false
   ) {module=SkeletalMeshLocation001}
   &InitialMeshOrientation(
    Orientation Axis=<0 0 0>
    Orientation Coordinate Space=::Local
    Orientation Vector=&RotateVectorByQuaternion(Vector To Rotate=<0 0 0>, Quaternion=$Output.SkeletalMeshLocation001.SampledRotation)
    Rotation=&Add_Vector(
     A=&Multiply_Vector(
      A=<0.33 0 0>
      B=&MakeVector(X=&MakeFloatFromInt(Int=$Particles.SpawnGroup), Y=1, Z=1)
     )
     B=&MakeVector(X=$Emitter.EmitterRand, Y=0, Z=0)
    )
    Use Orientation Vector=true
    Use Rotation Vector=true
   )
  ]
  particleUpdateScript [
   &ParticleState()
   &SkeletalMeshLocation(
    Inherit Velocity=false
    MaskSpawnGroup=false
    Mesh Sampling Type=::NewEnumerator5
    Sampled Orientation Coordinate Space=::Local
    Sampled Position Offset Coordinate Space:ENiagaraMeshTransforms=::NewEnumerator3
    Sampled Velocity Coordinate Space:ENiagaraCoordinateSpace=::Local
    Skeletal Mesh=$System.Skeletal Mesh
    Socket Sampling Mode=::NewEnumerator5
    Use Sampled Mesh Attribute as Ribbon ID=false
   )
   &ScaleColor(
    Scale Alpha=&FloatFromCurve(FloatCurve=DataInterface() {propertyValues {Curve {keys [
         {interpMode=RCIM_Cubic, tangentMode=RCTM_User, time=0, value=0}
         {arriveTangent=0.9375, interpMode=RCIM_Cubic, leaveTangent=0.9375, time=0.2, value=1}
         {arriveTangent=-1.25, interpMode=RCIM_Cubic, leaveTangent=-1.25, time=0.8, value=0.75}
         <at 1 0 RCIM_Cubic>
        ]}}})
    Scale RGB=&MakeVectorFromLinearColorRGB(Color=$System.ToDColor)
    ScaleA=true
    ScaleRGB=true
   )
   &ScaleMeshSize(Scale Factor=&VectorFromFloat(Value=&FloatFromCurve(FloatCurve=DataInterface() {propertyValues {Curve {keys [{interpMode=RCIM_Cubic, tangentMode=RCTM_User, time=0, value=1}, {interpMode=RCIM_Cubic, tangentMode=RCTM_User, time=1, value=0}]}}})))
   &UpdateMeshOrientation(Rotation Rate=-0.35, Rotation Vector=<1 0 0>)
   &DynamicMaterialParameters(
    Index 0 Param 2=&FloatFromCurve(FloatCurve=DataInterface() {propertyValues {Curve {keys [<at 0 1>,<at 1 1.5>]}}}, Scale Curve=0.3)
    Index 0 Param 4=&RandomRangeFloat(Minimum=0, Maximum=1)
    Index 1 Param 1=1
    Index 1 Param 2=&RandomRangeFloat(Minimum=0.8, Maximum=1.2)
    Index 1 Param 4=&RandomRangeFloat(Minimum=0, Maximum=1)
    Index 2 Param 2=&RandomRangeFloat(Minimum=-1, Maximum=-1.2)
    Index 2 Param 4=&RandomRangeFloat(Minimum=-2, Maximum=-3)
    Index 3 Param 1=1
    Index 3 Param 2=1.25
    Module.Param0WriteEnabled=true
    Module.Param0WriteEnabled1=true
    Module.Param0WriteEnabled2=true
    Module.Param0WriteEnabled3=true
    Module.Param1WriteEnabled=true
    Module.Param1WriteEnabled1=true
    Module.Param1WriteEnabled2=true
    Module.Param1WriteEnabled3=true
    Module.Param2WriteEnabled=true
    Module.Param2WriteEnabled1=true
    Module.Param2WriteEnabled2=true
    Module.Param2WriteEnabled3=false
    Module.Param3WriteEnabled=true
    Module.Param3WriteEnabled1=true
    Module.Param3WriteEnabled2=true
    Module.Param3WriteEnabled3=false
    Write Parameter Index 1=true
    Write Parameter Index 2=true
    Write Parameter Index 3=true
   )
  ]
  rendererClasses [&NiagaraMeshRendererProperties]
  renderers [{rendererClass=&NiagaraMeshRendererProperties, rendererIndex=0}]
 }
 RadialSprite {
  emitterUpdateScript [
   &EmitterState()
   &SpawnBurst_Instantaneous(Spawn Count=1, Use Spawn Probability=false)
  ]
  particleSpawnScript [&InitializeParticle(
    Lifetime Max=1.2
    Lifetime Min=0.8
    Lifetime Mode=::NewEnumerator1
    Sprite Size Mode=::NewEnumerator1
    Uniform Sprite Size Max=60
    Uniform Sprite Size Min=60
    UsePositionOffset=false
    Write Material Random=true
   )]
  particleUpdateScript [
   &ParticleState()
   &SkeletalMeshLocation(
    Inherit Velocity=false
    MaskSpawnGroup=false
    Mesh Sampling Type=::NewEnumerator5
    Sampled Orientation Coordinate Space=::Local
    Sampled Position Offset Coordinate Space:ENiagaraMeshTransforms=::NewEnumerator3
    Sampled Velocity Coordinate Space:ENiagaraCoordinateSpace=::Local
    Skeletal Mesh=$System.Skeletal Mesh
    Socket Sampling Mode=::NewEnumerator5
    Use Sampled Mesh Attribute as Ribbon ID=false
   )
   &FN_Radial_Material(
    Base Sprite Offset X:NiagaraFloat=&RandomRangeFloat(Minimum=0, Maximum=1)
    Base Sprite Offset Y:NiagaraFloat=&RandomRangeFloat(Minimum=0, Maximum=1)
    Base Sprite Tile U=2
    Depth Fade=5
    Radial Mask Strength=0.8
    Sprite Extrusion Speed:NiagaraFloat=&RandomRangeFloat(Minimum=-0.2, Maximum=-0.4)
    Sprite Spin Speed:NiagaraFloat=&RandomRangeFloat(Minimum=-0.6, Maximum=-0.8)
   )
   &ScaleSpriteSize(Scale Factor=&Vector2DFromFloat(Value=&FloatFromCurve(
      FloatCurve=DataInterface() {propertyValues {Curve {keys [{interpMode=RCIM_Cubic, tangentMode=RCTM_User, time=0, value=1}, {interpMode=RCIM_Cubic, tangentMode=RCTM_User, time=1, value=0}]}}}
      Scale Curve=7
     )))
   &ScaleColor(
    Scale Alpha=&FloatFromCurve(
     FloatCurve=DataInterface() {propertyValues {Curve {keys [
         {arriveTangent=1.3793, interpMode=RCIM_Cubic, leaveTangent=1.3793, tangentMode=RCTM_User, time=0, value=0}
         {arriveTangent=2.0253, interpMode=RCIM_Cubic, leaveTangent=2.0253, tangentMode=RCTM_User, time=0.3, value=1}
         {arriveTangent=-1.4286, interpMode=RCIM_Cubic, leaveTangent=-1.4286, time=0.8, value=0.75}
         <at 1 0 RCIM_Cubic>
        ]}}}
     Scale Curve=&Lerp_Float(A=3, B=5, Alpha=$User.Variant)
    )
    Scale RGB=&MakeVectorFromLinearColorRGB(Color=$System.ToDColor)
    ScaleA=true
    ScaleRGB=true
   )
   /SetVariables(Particles.Default:NiagaraInt32=&MakeCustomIntFromBool(Bool=&SetBoolByFloatComparison(A=$User.Variant, B=0), True INT=1, False INT=0)) {isSetParameters=true}
   /SetVariables(Particles.Neon:NiagaraInt32=&MakeCustomIntFromBool(Bool=&SetBoolByFloatComparison(A=$User.Variant, B=0), True INT=0, False INT=1)) {isSetParameters=true}
   /SetVariables(
    Particles.SpriteColor_02:LinearColor=&MultiplyLinearColor(
     Linear Color=&LerpLinearColors(StartColor=<rgba 0.0395 0.2016 0.1746 1>, EndColor=<rgba 0.0202 0.0202 0.2016 1>, LerpFactor=$User.Variant)
     Scale Factor=&MakeFloat4FromVectorAndFloat(
      "Vector (RGB)"=&MakeVectorFromLinearColorRGB(Color=$System.ToDColor)
      "Float (Alpha)"=&FloatFromCurve(
       FloatCurve=DataInterface() {propertyValues {Curve {keys [
           {arriveTangent=9.0725, interpMode=RCIM_Cubic, leaveTangent=9.0725, tangentMode=RCTM_User, time=0, value=0}
           {arriveTangent=-2.5249, interpMode=RCIM_Cubic, leaveTangent=-2.5249, tangentMode=RCTM_User, time=0.1408, value=0.9847}
           {arriveTangent=-0.0405, interpMode=RCIM_Cubic, leaveTangent=-0.0405, tangentMode=RCTM_User, time=1, value=0}
          ]}}}
       Scale Curve=0.8
      )
     )
    )
    Particles.SpriteSize_02:Vector2f=&Multiply_Vector2(
     A=&Vector2DFromFloat(Value=30)
     B=&Vector2DFromFloat(Value=&FloatFromCurve(
       FloatCurve=DataInterface() {propertyValues {Curve {keys [
           <at 0 0 RCIM_Cubic>
           {arriveTangent=2, interpMode=RCIM_Cubic, leaveTangent=0.0802, tangentMode=RCTM_Break, time=0.1, value=1}
           {arriveTangent=-0.1222, interpMode=RCIM_Cubic, leaveTangent=-1.25, tangentMode=RCTM_Break, time=0.5, value=1}
           <at 1 0 RCIM_Cubic>
          ]}}}
       Scale Curve=7.5
      ))
    )
   ) {isSetParameters=true}
  ]
  rendererClasses [&NiagaraSpriteRendererProperties]
  renderers [
   {rendererClass=&NiagaraSpriteRendererProperties}
   {rendererClass=&NiagaraSpriteRendererProperties, rendererIndex=1}
   {rendererClass=&NiagaraSpriteRendererProperties, rendererIndex=2}
  ]
 }
}
summary {
 systemName=NS_Glider_GraffitiTon_Deploy_Charge
 emitters [
  {emitterName=Blobs, rendererClasses [&NiagaraSpriteRendererProperties]}
  {emitterName=RadialSprite, rendererClasses [&NiagaraSpriteRendererProperties]}
  {emitterName=MeshCards, rendererClasses [&NiagaraMeshRendererProperties]}
  {emitterName=BigSplash, rendererClasses [&NiagaraSpriteRendererProperties]}
 ]
 userVariables [
  User.bIsFrontEnd {type=NiagaraBool}
  User.bIsFrontEnd {type=NiagaraFloat, defaultValue=0, description=""}
  User.bIsFrontEndPreview {type=NiagaraBool}
  User.bisFullyDeployed {type=NiagaraBool}
  User.ForwardDot {type=NiagaraFloat, defaultValue=0, description=""}
  User.RibbonIDOffset {type=NiagaraInt32}
  User.RightDot {type=NiagaraFloat, defaultValue=0, description=""}
  User.Variant {type=NiagaraFloat, defaultValue=0, description=""}
 ]
}
